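# Sound cue for the swing of the male Kane cosmetic pickaxe.
# A quality switch picks branches, each mixing a close-attenuated swing with a medium-attenuated Fire Axe distant swing.
# The first branch randomizes six Kane waves and six distant waves under pitch modulation.
# The other branches use fewer waves and substitute Gingerbread pickaxe swings for the Kane ones, with the last two levels sharing one mixer.
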SoundCue /BRCosmetics/Sounds/PickAxes/Kane/Pickaxe_Kane_Male_Swing
root=QualityLevel_0
nodes [
 {id=Attenuation_0, children [Modulator_3], properties {attenuationSettings=/Game/Sounds/Attenuation/Fort_Weapon_Pickaxe_Close_Att}}
 {id=Attenuation_1, children [Modulator_4], properties {attenuationSettings=/Game/Sounds/Attenuation/Fort_Weapon_Pickaxe_Medium_Att}}
 {id=Attenuation_2, children [Random_1], properties {attenuationSettings=&Fort_Weapon_Pickaxe_Close_Att}}
 {id=Attenuation_3, children [Random_0], properties {attenuationSettings=&Fort_Weapon_Pickaxe_Medium_Att}}
 {id=Attenuation_4, children [Random_6], properties {attenuationSettings=&Fort_Weapon_Pickaxe_Close_Att}}
 {id=Attenuation_5, children [Random_3], properties {attenuationSettings=&Fort_Weapon_Pickaxe_Medium_Att}}
 {id=Mixer_0, children [Attenuation_0, Attenuation_1], properties {inputVolume [1,1]}}
 {id=Mixer_1, children [Attenuation_2, Attenuation_3], properties {inputVolume [1,1]}}
 {id=Mixer_2, children [Attenuation_4, Attenuation_5], properties {inputVolume [1,1]}}
 {id=Modulator_1, children [Mixer_1], properties {pitchMax=1.03, pitchMin=0.97}}
 {id=Modulator_2, children [Mixer_2]}
 {id=Modulator_3, children [Random_7], properties {pitchMax=1.2, pitchMin=0.9}}
 {id=Modulator_4, children [WavePlayer_28], properties {pitchMax=1.2, pitchMin=0.9}}
 {id=QualityLevel_0, children [Modulator_1,Modulator_2,Mixer_0,Mixer_0]}
 {
  id=Random_0
  children [WavePlayer_13, WavePlayer_1, WavePlayer_2, WavePlayer_12, WavePlayer_4, WavePlayer_16]
  properties {weights [1,1,1,1,1,1]}
 }
 {
  id=Random_1
  children [WavePlayer_3, WavePlayer_6, WavePlayer_7, WavePlayer_8, WavePlayer_9, WavePlayer_25]
  properties {weights [1,1,1,1,1,1]}
 }
 {id=Random_3, children [WavePlayer_19, WavePlayer_18, WavePlayer_20], properties {weights [1,1,1]}}
 {id=Random_6, children [WavePlayer_10, WavePlayer_11, WavePlayer_14], properties {weights [1,1,1]}}
 {id=Random_7, children [WavePlayer_15, WavePlayer_17], properties {weights [1,1]}}
 {id=WavePlayer_1, properties {soundWaveAssetPtr=/Game/Athena/Sounds/Weapons/Fire_Axe_Swing_Distant_01}}
 {id=WavePlayer_10, properties {soundWaveAssetPtr=/BRCosmetics/Sounds/PickAxes/Gingerbread/Pickaxe_Gingerbread_Swing_01}}
 {id=WavePlayer_11, properties {soundWaveAssetPtr=/BRCosmetics/Sounds/PickAxes/Gingerbread/Pickaxe_Gingerbread_Swing_02}}
 {id=WavePlayer_12, properties {soundWaveAssetPtr=/Game/Athena/Sounds/Weapons/Fire_Axe_Swing_Distant_03}}
 {id=WavePlayer_13, properties {soundWaveAssetPtr=/Game/Athena/Sounds/Weapons/Fire_Axe_Swing_Distant_04}}
 {id=WavePlayer_14, properties {soundWaveAssetPtr=/BRCosmetics/Sounds/PickAxes/Gingerbread/Pickaxe_Gingerbread_Swing_03}}
 {id=WavePlayer_15, properties {soundWaveAssetPtr=&Pickaxe_Gingerbread_Swing_01}}
 {id=WavePlayer_16, properties {soundWaveAssetPtr=/Game/Athena/Sounds/Weapons/Fire_Axe_Swing_Distant_06}}
 {id=WavePlayer_17, properties {soundWaveAssetPtr=&Pickaxe_Gingerbread_Swing_02}}
 {id=WavePlayer_18, properties {soundWaveAssetPtr=&Fire_Axe_Swing_Distant_03}}
 {id=WavePlayer_19, properties {soundWaveAssetPtr=&Fire_Axe_Swing_Distant_04}}
 {id=WavePlayer_2, properties {soundWaveAssetPtr=/Game/Athena/Sounds/Weapons/Fire_Axe_Swing_Distant_02}}
 {id=WavePlayer_20, properties {soundWaveAssetPtr=&Fire_Axe_Swing_Distant_06}}
 {id=WavePlayer_25, properties {soundWaveAssetPtr=&/PickAxe_KaneMale_Swing_06}}
 {id=WavePlayer_28, properties {soundWaveAssetPtr=&Fire_Axe_Swing_Distant_06}}
 {id=WavePlayer_3, properties {soundWaveAssetPtr=&/PickAxe_KaneMale_Swing_01}}
 {id=WavePlayer_4, properties {soundWaveAssetPtr=/Game/Athena/Sounds/Weapons/Fire_Axe_Swing_Distant_05}}
 {id=WavePlayer_6, properties {soundWaveAssetPtr=&/PickAxe_KaneMale_Swing_02}}
 {id=WavePlayer_7, properties {soundWaveAssetPtr=&/PickAxe_KaneMale_Swing_03}}
 {id=WavePlayer_8, properties {soundWaveAssetPtr=&/PickAxe_KaneMale_Swing_04}}
 {id=WavePlayer_9, properties {soundWaveAssetPtr=&/PickAxe_KaneMale_Swing_05}}
]
properties {
 audioPropertiesSheet=/BRCosmetics/Sounds/Settings/PropertySheets/Pickaxe/APS_Cosmetic_Pickaxe_Swing
 bPrimeOnLoad=true
 maxDistance=3000
 soundClassObject=&Fort_WeaponMelee_PSM
 volumeMultiplier=0.5
}
